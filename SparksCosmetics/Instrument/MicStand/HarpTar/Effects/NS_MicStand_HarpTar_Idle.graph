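# Niagara idle effect for the HarpTar mic stand cosmetic, placing glow and sparkle sprites around the skeletal mesh passed as a user parameter.
# Glow picks one of three bones from a weighted array, then adds a per-bone curve, line or ring offset, with pink-purple fading color.
# Sparkle spawns sparsely, starting from a curve-shaped random distribution.
# DebugGlows, DebugSparkle and GlowPersistent are disabled variants, the debug ones previewing placement shapes.
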NiagaraSystem /SparksCosmetics/Instrument/MicStand/HarpTar/Effects/NS_MicStand_HarpTar_Idle
dependencies {
 usedDataInterfaces [
  &NiagaraDataInterfaceVectorCurve
  &NiagaraDataInterfaceArrayInt32
  &NiagaraDataInterfaceCurve
  &NiagaraDataInterfaceColorCurve
 ]
 usedDynamicInputs []
 usedModules [&SystemState]
}
emitters {
 DebugGlows {
  bEnabled=false
  emitterUpdateScript [
   &EmitterState()
   &SpawnRate(SpawnRate=50, Use Spawn Probability=false)
  ]
  particleSpawnScript [
   &InitializeParticle(Lifetime=0.5, Sprite Size Mode=::NewEnumerator3, Uniform Sprite Size=3, UsePositionOffset=false, Write Material Random=true)
   /SetVariables(
    Transient.BaseWidth:NiagaraFloat=&RandomRangeFloat(Minimum=0, Maximum=1)
    Transient.PickBone:NiagaraInt32=&RandomRangeInt(Minimum=0, Maximum=2)
   ) {isSetParameters=true, module=SetVariables_#1}
   /SetVariables(Transient.PickOffset:Vector3f=&self:SelectOffsetByBone(
     Offset 0:Vector3f=&Add_Vector(
      A=&VectorFromCurve(
       VectorCurve=DataInterface() {propertyValues {
         XCurve {keys [
           {arriveTangent=17.51, interpMode=RCIM_Cubic, leaveTangent=17.51, tangentMode=RCTM_User, time=0, value=15}
           {arriveTangentWeight=1.4587, interpMode=RCIM_Cubic, leaveTangentWeight=0.241, tangentMode=RCTM_User, time=0.8582, value=21.9634}
           {arriveTangent=-17.9309, interpMode=RCIM_Cubic, leaveTangent=-17.9309, tangentMode=RCTM_User, time=1, value=20}
          ]}
         YCurve {keys [<at 0 5>,<at 1 25>]}
         ZCurve {keys [<at 0 0>]}
        }}
       CurveIndex=&RandomRangeFloat(Minimum=0, Maximum=1)
      )
      B=&VectorFromCurve(
       VectorCurve=DataInterface() {propertyValues {
         XCurve {keys [
           {arriveTangent=9.6911, interpMode=RCIM_Cubic, leaveTangent=9.6911, tangentMode=RCTM_User, time=0, value=-2}
           <at 0.5 0 RCIM_Cubic>
           {arriveTangent=-10.0913, interpMode=RCIM_Cubic, leaveTangent=-10.0913, tangentMode=RCTM_User, time=1, value=-2}
          ]}
         YCurve {keys [<at 0 0>]}
         ZCurve {keys [<at 0 7>,<at 1 -7>]}
        }}
       CurveIndex=&RandomRangeFloat(Minimum=0, Maximum=1)
      )
     )
     Offset 1:Vector3f=&Add_Vector(
      A=&RandomRangeVector(Minimum=<0 0 -2>, Maximum=<0 0 2>)
      B=&Lerp_Vector(A=<21 16 0>, B=<37 22 0>, Alpha=&RandomRangeFloat(Minimum=0, Maximum=1))
     )
     Offset 2:Vector3f=&Add_Vector(
      A=&Lerp_Vector(A=<0 0 0>, B=<0 2 0>, Alpha=$Transient.BaseWidth)
      B=&MakeVector(
       X=&Lerp_Float(A=5, B=13, Alpha=$Transient.BaseWidth)
       Y=&Sine(
        Normalized Angle=$Particles.MaterialRandom
        Period=1
        Scale=&Lerp_Float(A=15, B=8, Alpha=$Transient.BaseWidth)
        Bias=0
       )
       Z=&Cosine(
        Normalized Angle=$Particles.MaterialRandom
        Period=1
        Scale=&Lerp_Float(A=15, B=8, Alpha=$Transient.BaseWidth)
        Bias=0
       )
      )
     )
     Pick Bone:NiagaraInt32=$Transient.PickBone
    )) {isSetParameters=true, module=SetVariables_#2}
   &SkeletalMeshLocation(
    Bone / Socket Index=$Transient.PickBone
    MaskSpawnGroup=false
    Mesh Sampling Type=::NewEnumerator5
    Sampled Position Offset=$Transient.PickOffset
    Skeletal Mesh=$User.Skeletal Mesh
    Socket Sampling Mode=::NewEnumerator5
   )
  ]
  particleUpdateScript [&ParticleState()]
  rendererClasses [&NiagaraSpriteRendererProperties]
  renderers [{rendererClass=&NiagaraSpriteRendererProperties}]
 }
 DebugSparkle {
  bEnabled=false
  emitterUpdateScript [
   &EmitterState()
   &SpawnRate(SpawnRate=200, Use Spawn Probability=false)
  ]
  particleSpawnScript [
   &InitializeParticle(Color=<rgba 1 1 1 0.0817>, Sprite Size Mode=::NewEnumerator3, Uniform Sprite Size=2, UsePositionOffset=false, Write Material Random=true)
   /SetVariables(Transient.ZOffset:NiagaraFloat=&RandomRangeFloat(Minimum=0, Maximum=1)) {enabled=false, isSetParameters=true}
   /SetVariables(Transient.ZOffset:NiagaraFloat=&FloatFromCurve(
     FloatCurve=DataInterface() {propertyValues {Curve {keys [
         <at 0 0 RCIM_Cubic>
         {arriveTangent=3.7175, interpMode=RCIM_Cubic, leaveTangent=3.7175, tangentMode=RCTM_User, time=0.5, value=0.5}
         <at 1 1>
        ]}}}
     CurveIndex=&RandomRangeFloat(Minimum=0, Maximum=1)
    )) {enabled=false, isSetParameters=true}
   /SetVariables(Transient.ZOffset:NiagaraFloat=&FloatFromCurve(
     FloatCurve=DataInterface() {propertyValues {Curve {keys [
         {arriveTangent=0.145, interpMode=RCIM_Cubic, leaveTangent=0.145, leaveTangentWeight=0.1403, tangentMode=RCTM_User, time=0, value=0}
         {arriveTangentWeight=0.0585, interpMode=RCIM_Cubic, leaveTangentWeight=0.1243, tangentMode=RCTM_User, time=0.16, value=0.02}
         {arriveTangent=2.8919, interpMode=RCIM_Cubic, leaveTangent=2.8919, tangentMode=RCTM_User, time=0.5, value=0.5}
         {arriveTangentWeight=0.1603, interpMode=RCIM_Cubic, leaveTangentWeight=0.0754, tangentMode=RCTM_User, time=0.84, value=0.98}
         {arriveTangent=0.134, tangentMode=RCTM_User, time=1, value=1}
        ]}}}
     CurveIndex=&RandomRangeFloat(Minimum=0, Maximum=1)
    )) {isSetParameters=true, module=SetVariables_#3}
   /SetVariables(Transient.ZOffsetSymm:NiagaraFloat=&FloatFromCurve(FloatCurve=DataInterface() {propertyValues {Curve {keys [<at 0 0>,<at 0.5 1>,<at 1 0>]}}}, CurveIndex=$Transient.ZOffset)) {isSetParameters=true, module=SetVariables_#4}
   /SetVariables(
    Transient.XOffsetCore:NiagaraFloat=&FloatFromCurve(
     FloatCurve=DataInterface() {propertyValues {Curve {keys [
         {
          arriveTangent=-342.3227
          interpMode=RCIM_Cubic
          leaveTangent=-342.3227
          leaveTangentWeight=0.013
          tangentMode=RCTM_User
          time=0
          value=-2.0158
         }
         {arriveTangentWeight=0.0902, interpMode=RCIM_Cubic, leaveTangentWeight=0.0902, tangentMode=RCTM_User, time=0.01, value=-4}
         {arriveTangent=129.5097, interpMode=RCIM_Cubic, leaveTangent=129.5097, tangentMode=RCTM_User, time=0.02, value=-3}
         {arriveTangent=9.5976, interpMode=RCIM_Cubic, leaveTangent=9.5976, tangentMode=RCTM_User, time=0.055, value=-1}
         {arriveTangent=0.1052, interpMode=RCIM_Cubic, leaveTangent=0.1052, tangentMode=RCTM_User, time=1, value=0}
        ]}}}
     CurveIndex=$Transient.ZOffsetSymm
    )
    Transient.XOffsetRandom:NiagaraFloat=&FloatFromCurve(
     FloatCurve=DataInterface() {propertyValues {Curve {keys [
         {arriveTangent=402.6897, interpMode=RCIM_Cubic, leaveTangent=402.6897, tangentMode=RCTM_User, time=0, value=0}
         {arriveTangent=10.6141, interpMode=RCIM_Cubic, leaveTangent=10.6141, tangentMode=RCTM_User, time=0.02, value=3}
         {arriveTangent=2.6952, interpMode=RCIM_Cubic, leaveTangent=2.6952, tangentMode=RCTM_User, time=0.1, value=3.5}
         <at 1 4.5>
        ]}}}
     CurveIndex=$Transient.ZOffsetSymm
     Scale Curve=&RandomRangeFloat(Minimum=-1, Maximum=1)
    )
    Transient.YOffset:NiagaraFloat=&FloatFromCurve(
     FloatCurve=DataInterface() {propertyValues {Curve {keys [
         {arriveTangent=402.4345, interpMode=RCIM_Cubic, leaveTangent=402.4345, tangentMode=RCTM_User, time=0, value=-19.0415}
         {arriveTangent=31.798, interpMode=RCIM_Cubic, leaveTangent=30.7708, tangentMode=RCTM_Break, time=0.1, value=-5.0969}
         {arriveTangent=6.7744, interpMode=RCIM_Cubic, leaveTangent=6.7744, time=0.5, value=0}
         {arriveTangentWeight=0.4433, interpMode=RCIM_Cubic, leaveTangentWeight=2.8431, tangentMode=RCTM_User, time=1, value=1}
        ]}}}
     CurveIndex=$Transient.ZOffsetSymm
    )
   ) {isSetParameters=true, module=SetVariables_#5}
   /SetVariables(
    Transient.XOffsetCore:NiagaraFloat=&FloatFromCurve(
     FloatCurve=DataInterface() {propertyValues {Curve {keys [
         {
          arriveTangent=-342.3227
          interpMode=RCIM_Cubic
          leaveTangent=-342.3227
          leaveTangentWeight=0.013
          tangentMode=RCTM_User
          time=0
          value=-2.0158
         }
         {arriveTangentWeight=0.0902, interpMode=RCIM_Cubic, leaveTangentWeight=0.0902, tangentMode=RCTM_User, time=0.01, value=-4}
         {arriveTangent=129.5097, interpMode=RCIM_Cubic, leaveTangent=129.5097, tangentMode=RCTM_User, time=0.02, value=-3}
         {arriveTangent=9.5976, interpMode=RCIM_Cubic, leaveTangent=9.5976, tangentMode=RCTM_User, time=0.055, value=-1}
         {arriveTangent=0.1052, interpMode=RCIM_Cubic, leaveTangent=0.1052, tangentMode=RCTM_User, time=1, value=0}
        ]}}}
     CurveIndex=$Transient.ZOffsetSymm
    )
    Transient.XOffsetRandom:NiagaraFloat=&FloatFromCurve(
     FloatCurve=DataInterface() {propertyValues {Curve {keys [
         {arriveTangent=402.6897, interpMode=RCIM_Cubic, leaveTangent=402.6897, tangentMode=RCTM_User, time=0, value=0}
         {arriveTangent=10.6141, interpMode=RCIM_Cubic, leaveTangent=10.6141, tangentMode=RCTM_User, time=0.02, value=3}
         {arriveTangent=2.6952, interpMode=RCIM_Cubic, leaveTangent=2.6952, tangentMode=RCTM_User, time=0.1, value=3.5}
         <at 1 4.5>
        ]}}}
     CurveIndex=$Transient.ZOffsetSymm
     Scale Curve=&RandomRangeFloat(Minimum=-1, Maximum=1)
    )
    Transient.YOffset:NiagaraFloat=&FloatFromCurve(
     FloatCurve=DataInterface() {propertyValues {Curve {keys [
         {arriveTangent=402.4345, interpMode=RCIM_Cubic, leaveTangent=402.4345, tangentMode=RCTM_User, time=0, value=-19.0415}
         {arriveTangent=31.798, interpMode=RCIM_Cubic, leaveTangent=30.7708, tangentMode=RCTM_Break, time=0.1, value=-5.0969}
         {arriveTangent=6.7744, interpMode=RCIM_Cubic, leaveTangent=6.7744, time=0.5, value=0}
         {arriveTangentWeight=0.4433, interpMode=RCIM_Cubic, leaveTangentWeight=2.8431, tangentMode=RCTM_User, time=1, value=1}
        ]}}}
     CurveIndex=$Transient.ZOffsetSymm
    )
   ) {enabled=false, isSetParameters=true}
   /SetVariables(
    Transient.XOffsetCore:NiagaraFloat=&FloatFromCurve(
     FloatCurve=DataInterface() {propertyValues {Curve {keys [
         {
          arriveTangent=-255.7173
          interpMode=RCIM_Cubic
          leaveTangent=-255.7173
          leaveTangentWeight=0.013
          tangentMode=RCTM_User
          time=0
          value=-2.0158
         }
         {arriveTangentWeight=0.0033, interpMode=RCIM_Cubic, leaveTangentWeight=0.0157, tangentMode=RCTM_User, time=0.01, value=-3.8825}
         {arriveTangent=3.5832, interpMode=RCIM_Cubic, leaveTangent=3.5832, tangentMode=RCTM_User, time=0.055, value=-0.5}
         {arriveTangent=0.1052, interpMode=RCIM_Cubic, leaveTangent=0.1052, tangentMode=RCTM_User, time=1, value=1}
        ]}}}
     CurveIndex=$Transient.ZOffsetSymm
    )
    Transient.YOffset:NiagaraFloat=&FloatFromCurve(
     FloatCurve=DataInterface() {propertyValues {Curve {keys [
         {arriveTangent=402.4345, interpMode=RCIM_Cubic, leaveTangent=402.4345, tangentMode=RCTM_User, time=0, value=-19.0415}
         {arriveTangent=31.798, interpMode=RCIM_Cubic, leaveTangent=30.7708, tangentMode=RCTM_Break, time=0.1, value=-5.0969}
         {arriveTangent=6.7744, interpMode=RCIM_Cubic, leaveTangent=6.7744, time=0.5, value=0}
         {arriveTangentWeight=0.4433, interpMode=RCIM_Cubic, leaveTangentWeight=2.8431, tangentMode=RCTM_User, time=1, value=1}
        ]}}}
     CurveIndex=$Transient.ZOffsetSymm
    )
   ) {enabled=false, isSetParameters=true}
   &SkeletalMeshLocation(
    MaskSpawnGroup=false
    Mesh Sampling Type=::NewEnumerator5
    Sampled Position Offset=&Add_Vector(
     A=<5 27 0>
     B=&MakeVector(
      X=&Add_Float(A=$Transient.XOffsetCore, B=$Transient.XOffsetRandom)
      Y=$Transient.YOffset
      Z=&FloatFromCurve(
       FloatCurve=DataInterface() {propertyValues {Curve {keys [
           {arriveTangent=-258.0267, interpMode=RCIM_Cubic, leaveTangent=-258.0267, tangentMode=RCTM_User, time=0, value=-5.5}
           <at 0.015 -9>
           <at 0.5 0>
           {arriveTangentWeight=0.5183, interpMode=RCIM_Cubic, leaveTangentWeight=0.016, tangentMode=RCTM_User, time=0.985, value=9}
           {arriveTangent=-441.0373, interpMode=RCIM_Cubic, leaveTangent=-441.0373, tangentMode=RCTM_User, time=1, value=5.5}
          ]}}}
       CurveIndex=$Transient.ZOffset
      )
     )
    )
    Skeletal Mesh=$User.Skeletal Mesh
    Socket Sampling Mode=::NewEnumerator5
   )
   &Color(Color=&ColorFromCurve(
     ColorCurve=DataInterface() {propertyValues {
       AlphaCurve {keys [<at 1 0.2 RCIM_Cubic>]}
       BlueCurve {keys [<at 0 0 RCIM_Cubic>,<at 0.23 0>,<at 0.25 1>,<at 0.27 0>,<at 0.4797 0>,<at 0.73 0.5>,<at 0.75 1>,<at 0.77 0.5>,<at 1 1 RCIM_Cubic>]}
       GreenCurve {keys [<at 0 0 RCIM_Cubic>,<at 0.23 0.5>,<at 0.25 1>,<at 0.27 0.5>,<at 0.4797 1>,<at 0.73 0.5>,<at 0.75 1>,<at 0.77 0.5>,<at 1 0 RCIM_Cubic>]}
       RedCurve {keys [<at 0 1 RCIM_Cubic>,<at 0.23 0.5>,<at 0.25 1>,<at 0.27 0.5>,<at 0.4797 0>,<at 0.73 0>,<at 0.75 1>,<at 0.77 0>,<at 1 0 RCIM_Cubic>]}
      }}
     CurveIndex=$Transient.ZOffset
    )) {enabled=false}
   &Color(Color=&ColorFromCurve(
     ColorCurve=DataInterface() {propertyValues {
       AlphaCurve {keys [<at 1 0.2 RCIM_Cubic>]}
       BlueCurve {keys [<at 0.8665 0 RCIM_Cubic>,<at 1 1 RCIM_Cubic>]}
       GreenCurve {keys [<at 0.8665 0 RCIM_Cubic>,<at 1 1 RCIM_Cubic>]}
       RedCurve {keys [<at 0.8665 1 RCIM_Cubic>,<at 1 1 RCIM_Cubic>]}
      }}
     CurveIndex=&Frac_Float(Float=&Multiply_Float(A=$Transient.ZOffsetSymm, B=5))
    )) {enabled=false, module=Color001}
   /SetVariables(Transient.debugValue=0.1) {isSetParameters=true}
   &Color(Color=&MakeCustomLinearColorFromBool(
     Bool=&Boolean_AndOperation(
      Bool1=&SetBoolByFloatComparison(A=$Transient.ZOffsetSymm, B=&Add_Float(A=$Transient.debugValue, B=-0.003))
      Bool2=&SetBoolByFloatComparison(A=$Transient.ZOffsetSymm, B=&Add_Float(A=$Transient.debugValue, B=0.003), Comparison Type=::NewEnumerator6)
     )
     True Color=<rgba 1 1 1 1>
     False Color=<rgba 1 0 0 1>
    )) {module=Color002}
  ]
  particleUpdateScript [
   &ParticleState()
   &CameraOffset(Camera Offset Amount=1)
  ]
  rendererClasses [&NiagaraSpriteRendererProperties]
  renderers [{rendererClass=&NiagaraSpriteRendererProperties}]
 }
 Glow {
  emitterUpdateScript [
   &EmitterState()
   &SpawnRate(SpawnRate=80, Use Spawn Probability=false)
  ]
  particleSpawnScript [
   &InitializeParticle(
    Color Maximum=<rgba 1 0.0643 0.2694 1>
    Color Minimum=<rgba 0.5583 0.0648 1 1>
    Color Mode=::NewEnumerator2
    Sprite Size Mode=::NewEnumerator3
    UsePositionOffset=false
    Write Material Random=true
   )
   /SetVariables(
    Transient.BaseWidth:NiagaraFloat=&RandomRangeFloat(Minimum=0, Maximum=1)
    Transient.PickBone:NiagaraInt32=&SelectIntFromArray(
     Array Sampling Mode=::NewEnumerator1
     Direct Array Index=?
     Int Selection Array=DataInterface() {propertyValues {IntData [0,0,1,2,2,2]}}
     Random Seed=0
    )
   ) {isSetParameters=true, module=SetVariables_#1}
   /SetVariables(Transient.PickOffset:Vector3f=&self:SelectOffsetByBone(
     Offset 0:Vector3f=&Add_Vector(
      A=&VectorFromCurve(
       VectorCurve=DataInterface() {propertyValues {
         XCurve {keys [
           {arriveTangent=11.1622, interpMode=RCIM_Cubic, leaveTangent=11.1622, tangentMode=RCTM_User, time=0, value=13.5}
           {
            arriveTangent=-14.7566
            arriveTangentWeight=0.2861
            interpMode=RCIM_Cubic
            leaveTangent=-14.7566
            leaveTangentWeight=0.0473
            tangentMode=RCTM_User
            time=0.8582
            value=22
           }
           {arriveTangent=-26.2212, interpMode=RCIM_Cubic, leaveTangent=-26.2212, tangentMode=RCTM_User, time=1, value=19}
          ]}
         YCurve {keys [<at 0 5>,<at 1 25>]}
         ZCurve {keys [<at 0 0>]}
        }}
       CurveIndex=&RandomRangeFloat(Minimum=0, Maximum=1)
      )
      B=&VectorFromCurve(
       VectorCurve=DataInterface() {propertyValues {
         XCurve {keys [
           {arriveTangent=9.6911, interpMode=RCIM_Cubic, leaveTangent=9.6911, tangentMode=RCTM_User, time=0, value=-2}
           <at 0.5 0 RCIM_Cubic>
           {arriveTangent=-10.0913, interpMode=RCIM_Cubic, leaveTangent=-10.0913, tangentMode=RCTM_User, time=1, value=-2}
          ]}
         YCurve {keys [<at 0 0>]}
         ZCurve {keys [<at 0 7>,<at 1 -7>]}
        }}
       CurveIndex=&RandomRangeFloat(Minimum=0, Maximum=1)
      )
     )
     Offset 1:Vector3f=&Add_Vector(
      A=&RandomRangeVector(Minimum=<0 0 -2>, Maximum=<0 0 2>)
      B=&Lerp_Vector(A=<22 14.5 0>, B=<37 20 0>, Alpha=&RandomRangeFloat(Minimum=0, Maximum=1))
     )
     Offset 2:Vector3f=&Add_Vector(
      A=&Lerp_Vector(A=<0 0 0>, B=<0 2 0>, Alpha=$Transient.BaseWidth)
      B=&MakeVector(
       X=&Lerp_Float(A=7, B=14, Alpha=$Transient.BaseWidth)
       Y=&Sine(
        Normalized Angle=$Particles.MaterialRandom
        Period=1
        Scale=&Lerp_Float(A=14.5, B=8.5, Alpha=$Transient.BaseWidth)
        Bias=0
       )
       Z=&Cosine(
        Normalized Angle=$Particles.MaterialRandom
        Period=1
        Scale=&Lerp_Float(A=14.5, B=8.5, Alpha=$Transient.BaseWidth)
        Bias=0
       )
      )
     )
     Pick Bone:NiagaraInt32=$Transient.PickBone
    )) {isSetParameters=true, module=SetVariables_#2}
   /SetVariables(Transient.PickOffset:Vector3f=&self:SelectOffsetByBone(
     Offset 0:Vector3f=&Add_Vector(
      A=&VectorFromCurve(
       VectorCurve=DataInterface() {propertyValues {
         XCurve {keys [
           {arriveTangent=17.51, interpMode=RCIM_Cubic, leaveTangent=17.51, tangentMode=RCTM_User, time=0, value=15}
           {arriveTangentWeight=1.4587, interpMode=RCIM_Cubic, leaveTangentWeight=0.241, tangentMode=RCTM_User, time=0.8582, value=21.9634}
           {arriveTangent=-17.9309, interpMode=RCIM_Cubic, leaveTangent=-17.9309, tangentMode=RCTM_User, time=1, value=20}
          ]}
         YCurve {keys [<at 0 5>,<at 1 25>]}
         ZCurve {keys [<at 0 0>]}
        }}
       CurveIndex=&RandomRangeFloat(Minimum=0, Maximum=1)
      )
      B=&VectorFromCurve(
       VectorCurve=DataInterface() {propertyValues {
         XCurve {keys [
           {arriveTangent=9.6911, interpMode=RCIM_Cubic, leaveTangent=9.6911, tangentMode=RCTM_User, time=0, value=-2}
           <at 0.5 0 RCIM_Cubic>
           {arriveTangent=-10.0913, interpMode=RCIM_Cubic, leaveTangent=-10.0913, tangentMode=RCTM_User, time=1, value=-2}
          ]}
         YCurve {keys [<at 0 0>]}
         ZCurve {keys [<at 0 7>,<at 1 -7>]}
        }}
       CurveIndex=&RandomRangeFloat(Minimum=0, Maximum=1)
      )
     )
     Offset 1:Vector3f=&Add_Vector(
      A=&RandomRangeVector(Minimum=<0 0 -2>, Maximum=<0 0 2>)
      B=&Lerp_Vector(A=<22 15 0>, B=<37 22 0>, Alpha=&RandomRangeFloat(Minimum=0, Maximum=1))
     )
     Offset 2:Vector3f=&Add_Vector(
      A=&Lerp_Vector(A=<0 0 0>, B=<0 2 0>, Alpha=$Transient.BaseWidth)
      B=&MakeVector(
       X=&Lerp_Float(A=5, B=13, Alpha=$Transient.BaseWidth)
       Y=&Sine(
        Normalized Angle=$Particles.MaterialRandom
        Period=1
        Scale=&Lerp_Float(A=15, B=10.5, Alpha=$Transient.BaseWidth)
        Bias=0
       )
       Z=&Cosine(
        Normalized Angle=$Particles.MaterialRandom
        Period=1
        Scale=&Lerp_Float(A=15, B=10.5, Alpha=$Transient.BaseWidth)
        Bias=0
       )
      )
     )
     Pick Bone:NiagaraInt32=$Transient.PickBone
    )) {enabled=false, isSetParameters=true}
   &SkeletalMeshLocation(
    Bone / Socket Index=$Transient.PickBone
    MaskSpawnGroup=false
    Mesh Sampling Type=::NewEnumerator5
    Sampled Position Offset=$Transient.PickOffset
    Skeletal Mesh=$User.Skeletal Mesh
    Socket Sampling Mode=::NewEnumerator5
   )
  ]
  particleUpdateScript [
   &ParticleState()
   &ScaleSpriteSize(Uniform Curve Sprite Scale=DataInterface() {propertyValues {Curve {
       defaultValue=0
       keys [
        {arriveTangent=3.1808, interpMode=RCIM_Cubic, leaveTangent=3.1808, tangentMode=RCTM_User, time=0, value=0}
        {interpMode=RCIM_Cubic, tangentMode=RCTM_User, time=1, value=1}
       ]
      }}})
   &CameraOffset(Camera Offset Amount=3)
   /SetVariables(Particles.Color=&RandomRangeLinearColor(
     Minimum=&ColorFromCurve(ColorCurve=DataInterface() {propertyValues {
        AlphaCurve {keys [{arriveTangent=2.6475, interpMode=RCIM_Cubic, time=0, value=0}, <at 0.5 1 RCIM_Cubic>, <at 1 0 RCIM_Cubic>]}
        BlueCurve {keys [<at 0 0.6096 RCIM_Cubic>, {arriveTangent=-0.3904, interpMode=RCIM_Cubic, leaveTangent=-0.3904, time=0.5, value=0.2972}, <at 1 0.2192>]}
        GreenCurve {keys [<at 0 0.5 RCIM_Cubic>, {arriveTangent=-0.5, interpMode=RCIM_Cubic, leaveTangent=-0.5, time=0.5, value=0.1}, <at 1 0>]}
        RedCurve {keys [<at 0 1 RCIM_Cubic>,<at 0.5 1 RCIM_Cubic>,<at 1 1>]}
       }})
     Maximum=&ColorFromCurve(ColorCurve=DataInterface() {propertyValues {
        AlphaCurve {keys [<at 0 0 RCIM_Cubic>,<at 0.5 1 RCIM_Cubic>,<at 1 0 RCIM_Cubic>]}
        BlueCurve {keys [<at 0 1 RCIM_Cubic>,<at 0.5 1 RCIM_Cubic>,<at 1 1>]}
        GreenCurve {keys [<at 0 0.5 RCIM_Cubic>, {arriveTangent=-0.5, interpMode=RCIM_Cubic, leaveTangent=-0.5, time=0.5, value=0.1}, <at 1 0>]}
        RedCurve {keys [<at 0 0.7643 RCIM_Cubic>, {arriveTangent=-0.2357, interpMode=RCIM_Cubic, leaveTangent=-0.2357, time=0.5, value=0.5758}, <at 1 0.5287>]}
       }})
    )) {enabled=false, isSetParameters=true, module=SetVariables_#6}
   /SetVariables(Particles.Color=&RandomRangeLinearColor(
     Minimum=&ColorFromCurve(ColorCurve=DataInterface() {propertyValues {
        AlphaCurve {keys [{arriveTangent=2.6475, interpMode=RCIM_Cubic, time=0, value=0}, <at 0.5 1 RCIM_Cubic>, <at 1 0 RCIM_Cubic>]}
        BlueCurve {keys [<at 0 0.1524 RCIM_Cubic>, {arriveTangent=-0.0428, interpMode=RCIM_Cubic, leaveTangent=-0.0428, time=0.5, value=0.1486}, <at 1 0.1096>]}
        GreenCurve {keys [<at 0 0.125 RCIM_Cubic>, {arriveTangent=-0.125, interpMode=RCIM_Cubic, leaveTangent=-0.125, time=0.5, value=0.05}, <at 1 0>]}
        RedCurve {keys [<at 0 0.25 RCIM_Cubic>, {arriveTangent=0.25, interpMode=RCIM_Cubic, leaveTangent=0.25, time=0.5, value=0.5}, <at 1 0.5>]}
       }})
     Maximum=&ColorFromCurve(ColorCurve=DataInterface() {propertyValues {
        AlphaCurve {keys [<at 0 0 RCIM_Cubic>,<at 0.5 1 RCIM_Cubic>,<at 1 0 RCIM_Cubic>]}
        BlueCurve {keys [<at 0 0.25 RCIM_Cubic>, {arriveTangent=0.25, interpMode=RCIM_Cubic, leaveTangent=0.25, time=0.5, value=0.5}, <at 1 0.5>]}
        GreenCurve {keys [<at 0 0.125 RCIM_Cubic>, {arriveTangent=-0.125, interpMode=RCIM_Cubic, leaveTangent=-0.125, time=0.5, value=0.05}, <at 1 0>]}
        RedCurve {keys [<at 0 0.1911 RCIM_Cubic>, {arriveTangent=0.0733, interpMode=RCIM_Cubic, leaveTangent=0.0733, time=0.5, value=0.2879}, <at 1 0.2643>]}
       }})
    )) {isSetParameters=true, module=SetVariables_#7}
   /SetVariables(Particles.Color=&RandomRangeLinearColor(
     Minimum=&ColorFromCurve(ColorCurve=DataInterface() {propertyValues {
        AlphaCurve {keys [
          {arriveTangent=2.6475, interpMode=RCIM_Cubic, time=0, value=0}
          {arriveTangentWeight=0.1389, interpMode=RCIM_Cubic, leaveTangentWeight=0.2778, tangentMode=RCTM_User, time=0.25, value=1}
          {arriveTangentWeight=0.2778, interpMode=RCIM_Cubic, leaveTangentWeight=0.1389, tangentMode=RCTM_User, time=0.75, value=1}
          <at 1 0 RCIM_Cubic>
         ]}
        BlueCurve {keys [<at 0 0.484 RCIM_Cubic>, {arriveTangent=-0.1996, interpMode=RCIM_Cubic, leaveTangent=-0.1996, time=0.5, value=0.327}, <at 1 0.2844>]}
        GreenCurve {keys [<at 0 0.259 RCIM_Cubic>, {arriveTangent=-0.259, interpMode=RCIM_Cubic, leaveTangent=-0.259, time=0.5, value=0.059}, <at 1 0>]}
        RedCurve {keys [<at 0 0.578 RCIM_Cubic>,<at 0.5 0.578 RCIM_Cubic>,<at 1 0.578>]}
       }})
     Maximum=&ColorFromCurve(ColorCurve=DataInterface() {propertyValues {
        AlphaCurve {keys [
          <at 0 0 RCIM_Cubic>
          {arriveTangentWeight=0.1389, interpMode=RCIM_Cubic, leaveTangentWeight=0.2778, tangentMode=RCTM_User, time=0.25, value=1}
          {arriveTangentWeight=0.2778, interpMode=RCIM_Cubic, leaveTangentWeight=0.1389, tangentMode=RCTM_User, time=0.75, value=1}
          <at 1 0 RCIM_Cubic>
         ]}
        BlueCurve {keys [<at 0 0.679 RCIM_Cubic>,<at 0.5 0.679 RCIM_Cubic>,<at 1 0.679>]}
        GreenCurve {keys [<at 0 0.259 RCIM_Cubic>, {arriveTangent=-0.259, interpMode=RCIM_Cubic, leaveTangent=-0.259, time=0.5, value=0.059}, <at 1 0>]}
        RedCurve {keys [<at 0 0.46 RCIM_Cubic>, {arriveTangent=-0.1215, interpMode=RCIM_Cubic, leaveTangent=-0.1215, time=0.5, value=0.366}, <at 1 0.3385>]}
       }})
    )) {enabled=false, isSetParameters=true, module=SetVariables_#8}
   /SetVariables(Particles.Color=&RandomRangeLinearColor(
     Minimum=&ColorFromCurve(ColorCurve=DataInterface() {propertyValues {
        AlphaCurve {keys [
          {arriveTangent=2.6475, interpMode=RCIM_Cubic, time=0, value=0}
          {arriveTangentWeight=0.1389, interpMode=RCIM_Cubic, leaveTangentWeight=0.2778, tangentMode=RCTM_User, time=0.25, value=1}
          {arriveTangentWeight=0.2778, interpMode=RCIM_Cubic, leaveTangentWeight=0.1389, tangentMode=RCTM_User, time=0.75, value=1}
          <at 1 0 RCIM_Cubic>
         ]}
        BlueCurve {keys [<at 0 0.421 RCIM_Cubic>, {arriveTangent=-0.2031, interpMode=RCIM_Cubic, leaveTangent=-0.2031, time=0.5, value=0.2337}, <at 1 0.2179>]}
        GreenCurve {keys [<at 0 0.139 RCIM_Cubic>, {arriveTangent=-0.139, interpMode=RCIM_Cubic, leaveTangent=-0.139, time=0.5, value=0.0259}, <at 1 0>]}
        RedCurve {keys [<at 0 0.367 RCIM_Cubic>, {arriveTangent=-0.117, interpMode=RCIM_Cubic, leaveTangent=-0.117, time=0.5, value=0.25}, <at 1 0.25>]}
       }})
     Maximum=&ColorFromCurve(ColorCurve=DataInterface() {propertyValues {
        AlphaCurve {keys [
          <at 0 0 RCIM_Cubic>
          {arriveTangentWeight=0.1389, interpMode=RCIM_Cubic, leaveTangentWeight=0.2778, tangentMode=RCTM_User, time=0.25, value=1}
          {arriveTangentWeight=0.2778, interpMode=RCIM_Cubic, leaveTangentWeight=0.1389, tangentMode=RCTM_User, time=0.75, value=1}
          <at 1 0 RCIM_Cubic>
         ]}
        BlueCurve {keys [<at 0 0.679 RCIM_Cubic>, {arriveTangent=-0.429, interpMode=RCIM_Cubic, leaveTangent=-0.429, time=0.5, value=0.25}, <at 1 0.25>]}
        GreenCurve {keys [<at 0 0.259 RCIM_Cubic>, {arriveTangent=-0.259, interpMode=RCIM_Cubic, leaveTangent=-0.259, time=0.5, value=0.0217}, <at 1 0>]}
        RedCurve {keys [<at 0 0.427 RCIM_Cubic>, {arriveTangent=-0.327, interpMode=RCIM_Cubic, leaveTangent=-0.327, time=0.5, value=0.113}, <at 1 0.1>]}
       }})
    )) {enabled=false, isSetParameters=true, module=SetVariables_#9}
   /SetVariables(Particles.Color=&ColorFromCurve(ColorCurve=DataInterface() {propertyValues {
       AlphaCurve {keys [<at 0 0 RCIM_Cubic>,<at 0.5 1 RCIM_Cubic>,<at 1 0 RCIM_Cubic>]}
       BlueCurve {keys [<at 0.5 0.3583 RCIM_Cubic>]}
       GreenCurve {keys [{arriveTangent=-0.5, interpMode=RCIM_Cubic, time=0.5, value=0.0181}]}
       RedCurve {keys [{arriveTangent=-0.2357, interpMode=RCIM_Cubic, time=0.5, value=0.1557}]}
      }})) {enabled=false, isSetParameters=true, module=SetVariables_#10}
  ]
  rendererClasses [&NiagaraSpriteRendererProperties]
  renderers [
   {rendererClass=&NiagaraSpriteRendererProperties}
   {rendererClass=&NiagaraSpriteRendererProperties, rendererIndex=1}
  ]
 }
 GlowPersistent {
  bEnabled=false
  emitterUpdateScript [
   &EmitterState()
   &SpawnBurst_Instantaneous(Spawn Count=60, Use Spawn Probability=false, UseLoopCountLimit=true)
  ]
  particleSpawnScript [
   &InitializeParticle(
    Color Maximum=<rgba 1 0.0643 0.2694 1>
    Color Minimum=<rgba 0.5583 0.0648 1 1>
    Color Mode=::NewEnumerator2
    Lifetime=4
    Sprite Size Mode=::NewEnumerator3
    UsePositionOffset=false
    Write Material Random=true
   )
   /SetVariables(
    Transient.BaseWidth:NiagaraFloat=&RandomRangeFloat(Minimum=0, Maximum=1)
    Transient.PickBone:NiagaraInt32=&SelectIntFromArray(
     Array Sampling Mode=::NewEnumerator1
     Direct Array Index=?
     Int Selection Array=DataInterface() {propertyValues {IntData [0,1,2,2]}}
     Random Seed=0
    )
   ) {isSetParameters=true, module=SetVariables_#1}
   /SetVariables(Transient.PickOffset:Vector3f=&self:SelectOffsetByBone(
     Offset 0:Vector3f=&Add_Vector(
      A=&VectorFromCurve(
       VectorCurve=DataInterface() {propertyValues {
         XCurve {keys [
           {arriveTangent=17.51, interpMode=RCIM_Cubic, leaveTangent=17.51, tangentMode=RCTM_User, time=0, value=15}
           {arriveTangentWeight=1.4587, interpMode=RCIM_Cubic, leaveTangentWeight=0.241, tangentMode=RCTM_User, time=0.8582, value=21.9634}
           {arriveTangent=-17.9309, interpMode=RCIM_Cubic, leaveTangent=-17.9309, tangentMode=RCTM_User, time=1, value=20}
          ]}
         YCurve {keys [<at 0 5>,<at 1 25>]}
         ZCurve {keys [<at 0 0>]}
        }}
       CurveIndex=&RandomRangeFloat(Minimum=0, Maximum=1)
      )
      B=&VectorFromCurve(
       VectorCurve=DataInterface() {propertyValues {
         XCurve {keys [
           {arriveTangent=9.6911, interpMode=RCIM_Cubic, leaveTangent=9.6911, tangentMode=RCTM_User, time=0, value=-2}
           <at 0.5 0 RCIM_Cubic>
           {arriveTangent=-10.0913, interpMode=RCIM_Cubic, leaveTangent=-10.0913, tangentMode=RCTM_User, time=1, value=-2}
          ]}
         YCurve {keys [<at 0 0>]}
         ZCurve {keys [<at 0 7>,<at 1 -7>]}
        }}
       CurveIndex=&RandomRangeFloat(Minimum=0, Maximum=1)
      )
     )
     Offset 1:Vector3f=&Add_Vector(
      A=&RandomRangeVector(Minimum=<0 0 -2>, Maximum=<0 0 2>)
      B=&Lerp_Vector(A=<22 15 0>, B=<37 22 0>, Alpha=&RandomRangeFloat(Minimum=0, Maximum=1))
     )
     Offset 2:Vector3f=&Add_Vector(
      A=&Lerp_Vector(A=<0 0 0>, B=<0 2 0>, Alpha=$Transient.BaseWidth)
      B=&MakeVector(
       X=&Lerp_Float(A=5, B=13, Alpha=$Transient.BaseWidth)
       Y=&Sine(
        Normalized Angle=$Particles.MaterialRandom
        Period=1
        Scale=&Lerp_Float(A=15, B=10.5, Alpha=$Transient.BaseWidth)
        Bias=0
       )
       Z=&Cosine(
        Normalized Angle=$Particles.MaterialRandom
        Period=1
        Scale=&Lerp_Float(A=15, B=10.5, Alpha=$Transient.BaseWidth)
        Bias=0
       )
      )
     )
     Pick Bone:NiagaraInt32=$Transient.PickBone
    )) {isSetParameters=true, module=SetVariables_#2}
   &SkeletalMeshLocation(
    Bone / Socket Index=$Transient.PickBone
    MaskSpawnGroup=false
    Mesh Sampling Type=::NewEnumerator5
    Sampled Position Offset=$Transient.PickOffset
    Skeletal Mesh=$User.Skeletal Mesh
    Socket Sampling Mode=::NewEnumerator5
   )
  ]
  particleUpdateScript [
   &ParticleState(
    Kill Particles When Lifetime Has Elapsed=false
    Let Infinitely Lived Particles Die When Emitter Deactivates=true
    Loop Particles Lifetime=true
   )
   &ScaleSpriteSize(
    Uniform Curve Index=&Frac_Float(Float=&Add_Float(A=$Particles.NormalizedAge, B=&RandomRangeFloat(Minimum=0, Maximum=1)))
    Uniform Curve Sprite Scale=DataInterface() {propertyValues {Curve {
       defaultValue=0
       keys [
        {interpMode=RCIM_Cubic, tangentMode=RCTM_User, time=0, value=0.5}
        {interpMode=RCIM_Cubic, tangentMode=RCTM_User, time=0.5, value=1}
        {tangentMode=RCTM_User, time=1, value=0.5}
       ]
      }}}
   )
   &CameraOffset(Camera Offset Amount=3)
   /SetVariables(Particles.Color=&RandomRangeLinearColor(
     Minimum=&ColorFromCurve(ColorCurve=DataInterface() {propertyValues {
        AlphaCurve {keys [{arriveTangent=2.6475, interpMode=RCIM_Cubic, time=0, value=0}, <at 0.5 1 RCIM_Cubic>, <at 1 0 RCIM_Cubic>]}
        BlueCurve {keys [<at 0 0.6096 RCIM_Cubic>, {arriveTangent=-0.3904, interpMode=RCIM_Cubic, leaveTangent=-0.3904, time=0.5, value=0.2972}, <at 1 0.2192>]}
        GreenCurve {keys [<at 0 0.5 RCIM_Cubic>, {arriveTangent=-0.5, interpMode=RCIM_Cubic, leaveTangent=-0.5, time=0.5, value=0.1}, <at 1 0>]}
        RedCurve {keys [<at 0 1 RCIM_Cubic>,<at 0.5 1 RCIM_Cubic>,<at 1 1>]}
       }})
     Maximum=&ColorFromCurve(ColorCurve=DataInterface() {propertyValues {
        AlphaCurve {keys [<at 0 0 RCIM_Cubic>,<at 0.5 1 RCIM_Cubic>,<at 1 0 RCIM_Cubic>]}
        BlueCurve {keys [<at 0 1 RCIM_Cubic>,<at 0.5 1 RCIM_Cubic>,<at 1 1>]}
        GreenCurve {keys [<at 0 0.5 RCIM_Cubic>, {arriveTangent=-0.5, interpMode=RCIM_Cubic, leaveTangent=-0.5, time=0.5, value=0.1}, <at 1 0>]}
        RedCurve {keys [<at 0 0.7643 RCIM_Cubic>, {arriveTangent=-0.2357, interpMode=RCIM_Cubic, leaveTangent=-0.2357, time=0.5, value=0.5758}, <at 1 0.5287>]}
       }})
    )) {enabled=false, isSetParameters=true, module=SetVariables_#6}
   /SetVariables(Particles.Color=&RandomRangeLinearColor(
     Minimum=&ColorFromCurve(ColorCurve=DataInterface() {propertyValues {
        AlphaCurve {keys [{arriveTangent=2.6475, interpMode=RCIM_Cubic, time=0, value=0}, <at 0.5 1 RCIM_Cubic>, <at 1 0 RCIM_Cubic>]}
        BlueCurve {keys [<at 0 0.1524 RCIM_Cubic>, {arriveTangent=-0.0428, interpMode=RCIM_Cubic, leaveTangent=-0.0428, time=0.5, value=0.1486}, <at 1 0.1096>]}
        GreenCurve {keys [<at 0 0.125 RCIM_Cubic>, {arriveTangent=-0.125, interpMode=RCIM_Cubic, leaveTangent=-0.125, time=0.5, value=0.05}, <at 1 0>]}
        RedCurve {keys [<at 0 0.25 RCIM_Cubic>, {arriveTangent=0.25, interpMode=RCIM_Cubic, leaveTangent=0.25, time=0.5, value=0.5}, <at 1 0.5>]}
       }})
     Maximum=&ColorFromCurve(ColorCurve=DataInterface() {propertyValues {
        AlphaCurve {keys [<at 0 0 RCIM_Cubic>,<at 0.5 1 RCIM_Cubic>,<at 1 0 RCIM_Cubic>]}
        BlueCurve {keys [<at 0 0.25 RCIM_Cubic>, {arriveTangent=0.25, interpMode=RCIM_Cubic, leaveTangent=0.25, time=0.5, value=0.5}, <at 1 0.5>]}
        GreenCurve {keys [<at 0 0.125 RCIM_Cubic>, {arriveTangent=-0.125, interpMode=RCIM_Cubic, leaveTangent=-0.125, time=0.5, value=0.05}, <at 1 0>]}
        RedCurve {keys [<at 0 0.1911 RCIM_Cubic>, {arriveTangent=0.0733, interpMode=RCIM_Cubic, leaveTangent=0.0733, time=0.5, value=0.2879}, <at 1 0.2643>]}
       }})
    )) {enabled=false, isSetParameters=true, module=SetVariables_#7}
   /SetVariables(Particles.Color=&RandomRangeLinearColor(
     Minimum=&ColorFromCurve(ColorCurve=DataInterface() {propertyValues {
        AlphaCurve {keys [
          {arriveTangent=2.6475, interpMode=RCIM_Cubic, time=0, value=0}
          {arriveTangentWeight=0.1389, interpMode=RCIM_Cubic, leaveTangentWeight=0.2778, tangentMode=RCTM_User, time=0.25, value=1}
          {arriveTangentWeight=0.2778, interpMode=RCIM_Cubic, leaveTangentWeight=0.1389, tangentMode=RCTM_User, time=0.75, value=1}
          <at 1 0 RCIM_Cubic>
         ]}
        BlueCurve {keys [<at 0 0.484 RCIM_Cubic>, {arriveTangent=-0.1996, interpMode=RCIM_Cubic, leaveTangent=-0.1996, time=0.5, value=0.327}, <at 1 0.2844>]}
        GreenCurve {keys [<at 0 0.259 RCIM_Cubic>, {arriveTangent=-0.259, interpMode=RCIM_Cubic, leaveTangent=-0.259, time=0.5, value=0.059}, <at 1 0>]}
        RedCurve {keys [<at 0 0.578 RCIM_Cubic>,<at 0.5 0.578 RCIM_Cubic>,<at 1 0.578>]}
       }})
     Maximum=&ColorFromCurve(ColorCurve=DataInterface() {propertyValues {
        AlphaCurve {keys [
          <at 0 0 RCIM_Cubic>
          {arriveTangentWeight=0.1389, interpMode=RCIM_Cubic, leaveTangentWeight=0.2778, tangentMode=RCTM_User, time=0.25, value=1}
          {arriveTangentWeight=0.2778, interpMode=RCIM_Cubic, leaveTangentWeight=0.1389, tangentMode=RCTM_User, time=0.75, value=1}
          <at 1 0 RCIM_Cubic>
         ]}
        BlueCurve {keys [<at 0 0.679 RCIM_Cubic>,<at 0.5 0.679 RCIM_Cubic>,<at 1 0.679>]}
        GreenCurve {keys [<at 0 0.259 RCIM_Cubic>, {arriveTangent=-0.259, interpMode=RCIM_Cubic, leaveTangent=-0.259, time=0.5, value=0.059}, <at 1 0>]}
        RedCurve {keys [<at 0 0.46 RCIM_Cubic>, {arriveTangent=-0.1215, interpMode=RCIM_Cubic, leaveTangent=-0.1215, time=0.5, value=0.366}, <at 1 0.3385>]}
       }})
    )) {enabled=false, isSetParameters=true, module=SetVariables_#8}
   /SetVariables(Particles.Color=&RandomRangeLinearColor(
     Minimum=&ColorFromCurve(ColorCurve=DataInterface() {propertyValues {
        AlphaCurve {keys [
          {arriveTangent=2.6475, interpMode=RCIM_Cubic, time=0, value=0}
          {arriveTangentWeight=0.1389, interpMode=RCIM_Cubic, leaveTangentWeight=0.2778, tangentMode=RCTM_User, time=0.25, value=1}
          {arriveTangentWeight=0.2778, interpMode=RCIM_Cubic, leaveTangentWeight=0.1389, tangentMode=RCTM_User, time=0.75, value=1}
          <at 1 0 RCIM_Cubic>
         ]}
        BlueCurve {keys [<at 0 0.421 RCIM_Cubic>, {arriveTangent=-0.2031, interpMode=RCIM_Cubic, leaveTangent=-0.2031, time=0.5, value=0.2337}, <at 1 0.2179>]}
        GreenCurve {keys [<at 0 0.139 RCIM_Cubic>, {arriveTangent=-0.139, interpMode=RCIM_Cubic, leaveTangent=-0.139, time=0.5, value=0.0259}, <at 1 0>]}
        RedCurve {keys [<at 0 0.367 RCIM_Cubic>, {arriveTangent=-0.117, interpMode=RCIM_Cubic, leaveTangent=-0.117, time=0.5, value=0.25}, <at 1 0.25>]}
       }})
     Maximum=&ColorFromCurve(ColorCurve=DataInterface() {propertyValues {
        AlphaCurve {keys [
          <at 0 0 RCIM_Cubic>
          {arriveTangentWeight=0.1389, interpMode=RCIM_Cubic, leaveTangentWeight=0.2778, tangentMode=RCTM_User, time=0.25, value=1}
          {arriveTangentWeight=0.2778, interpMode=RCIM_Cubic, leaveTangentWeight=0.1389, tangentMode=RCTM_User, time=0.75, value=1}
          <at 1 0 RCIM_Cubic>
         ]}
        BlueCurve {keys [<at 0 0.679 RCIM_Cubic>, {arriveTangent=-0.429, interpMode=RCIM_Cubic, leaveTangent=-0.429, time=0.5, value=0.25}, <at 1 0.25>]}
        GreenCurve {keys [<at 0 0.259 RCIM_Cubic>, {arriveTangent=-0.259, interpMode=RCIM_Cubic, leaveTangent=-0.259, time=0.5, value=0.0217}, <at 1 0>]}
        RedCurve {keys [<at 0 0.427 RCIM_Cubic>, {arriveTangent=-0.327, interpMode=RCIM_Cubic, leaveTangent=-0.327, time=0.5, value=0.113}, <at 1 0.1>]}
       }})
    )) {enabled=false, isSetParameters=true, module=SetVariables_#9}
   /SetVariables(Particles.Color=&ColorFromCurve(ColorCurve=DataInterface() {propertyValues {
       AlphaCurve {keys [<at 0 0 RCIM_Cubic>,<at 0.5 1 RCIM_Cubic>,<at 1 0 RCIM_Cubic>]}
       BlueCurve {keys [<at 0.5 0.3583 RCIM_Cubic>]}
       GreenCurve {keys [{arriveTangent=-0.5, interpMode=RCIM_Cubic, time=0.5, value=0.0181}]}
       RedCurve {keys [{arriveTangent=-0.2357, interpMode=RCIM_Cubic, time=0.5, value=0.1557}]}
      }})) {enabled=false, isSetParameters=true, module=SetVariables_#10}
  ]
  rendererClasses [&NiagaraSpriteRendererProperties]
  renderers [
   {rendererClass=&NiagaraSpriteRendererProperties}
   {rendererClass=&NiagaraSpriteRendererProperties, rendererIndex=1}
  ]
 }
 Sparkle {
  emitterUpdateScript [
   &EmitterState()
   &SpawnRate(Spawn Probability=0.5, SpawnRate=2, Use Spawn Probability=true)
  ]
  particleSpawnScript [
   &InitializeParticle(
    Lifetime Max=1
    Lifetime Min=0.5
    Lifetime Mode=::NewEnumerator1
    Sprite Rotation Mode=::NewEnumerator2
    Sprite Size Mode=::NewEnumerator1
    Uniform Sprite Size Max=3
    Uniform Sprite Size Min=1.5
    UsePositionOffset=false
    Write Material Random=true
   )
   /SetVariables(Transient.ZOffset:NiagaraFloat=&FloatFromCurve(
     FloatCurve=DataInterface() {propertyValues {Curve {keys [
         {arriveTangent=0.145, interpMode=RCIM_Cubic, leaveTangent=0.145, leaveTangentWeight=0.1403, tangentMode=RCTM_User, time=0, value=0}
         {arriveTangentWeight=0.0585, interpMode=RCIM_Cubic, leaveTangentWeight=0.1243, tangentMode=RCTM_User, time=0.16, value=0.02}
         {arriveTangent=2.8919, interpMode=RCIM_Cubic, leaveTangent=2.8919, tangentMode=RCTM_User, time=0.5, value=0.5}
         {arriveTangentWeight=0.1603, interpMode=RCIM_Cubic, leaveTangentWeight=0.0754, tangentMode=RCTM_User, time=0.84, value=0.98}
         {arriveTangent=0.134, tangentMode=RCTM_User, time=1, value=1}
        ]}}}
     CurveIndex=&RandomRangeFloat(Minimum=0, Maximum=1)
    )) {isSetParameters=true, module=SetVariables_#3}
   /SetVariables(Transient.ZOffsetSymm:NiagaraFloat=&FloatFromCurve(FloatCurve=DataInterface() {propertyValues {Curve {keys [<at 0 0>,<at 0.5 1>,<at 1 0>]}}}, CurveIndex=$Transient.ZOffset)) {isSetParameters=true, module=SetVariables_#4}
   /SetVariables(
    Transient.XOffsetCore:NiagaraFloat=&FloatFromCurve(
     FloatCurve=DataInterface() {propertyValues {Curve {keys [
         {
          arriveTangent=-342.3227
          interpMode=RCIM_Cubic
          leaveTangent=-342.3227
          leaveTangentWeight=0.013
          tangentMode=RCTM_User
          time=0
          value=-2.0158
         }
         {arriveTangentWeight=0.0902, interpMode=RCIM_Cubic, leaveTangentWeight=0.0902, tangentMode=RCTM_User, time=0.01, value=-4}
         {arriveTangent=129.5097, interpMode=RCIM_Cubic, leaveTangent=129.5097, tangentMode=RCTM_User, time=0.02, value=-3}
         {arriveTangent=9.5976, interpMode=RCIM_Cubic, leaveTangent=9.5976, tangentMode=RCTM_User, time=0.055, value=-1}
         {arriveTangent=0.1052, interpMode=RCIM_Cubic, leaveTangent=0.1052, tangentMode=RCTM_User, time=1, value=0}
        ]}}}
     CurveIndex=$Transient.ZOffsetSymm
    )
    Transient.XOffsetRandom:NiagaraFloat=&FloatFromCurve(
     FloatCurve=DataInterface() {propertyValues {Curve {keys [
         {arriveTangent=402.6897, interpMode=RCIM_Cubic, leaveTangent=402.6897, tangentMode=RCTM_User, time=0, value=0}
         {arriveTangent=10.6141, interpMode=RCIM_Cubic, leaveTangent=10.6141, tangentMode=RCTM_User, time=0.02, value=3}
         {arriveTangent=2.6952, interpMode=RCIM_Cubic, leaveTangent=2.6952, tangentMode=RCTM_User, time=0.1, value=3.5}
         <at 1 4.5>
        ]}}}
     CurveIndex=$Transient.ZOffsetSymm
     Scale Curve=&RandomRangeFloat(Minimum=-1, Maximum=1)
    )
    Transient.YOffset:NiagaraFloat=&FloatFromCurve(
     FloatCurve=DataInterface() {propertyValues {Curve {keys [
         {arriveTangent=402.4345, interpMode=RCIM_Cubic, leaveTangent=402.4345, tangentMode=RCTM_User, time=0, value=-19.0415}
         {arriveTangent=31.798, interpMode=RCIM_Cubic, leaveTangent=30.7708, tangentMode=RCTM_Break, time=0.1, value=-5.0969}
         {arriveTangent=6.7744, interpMode=RCIM_Cubic, leaveTangent=6.7744, time=0.5, value=0}
         {arriveTangentWeight=0.4433, interpMode=RCIM_Cubic, leaveTangentWeight=2.8431, tangentMode=RCTM_User, time=1, value=1}
        ]}}}
     CurveIndex=$Transient.ZOffsetSymm
    )
   ) {isSetParameters=true, module=SetVariables_#5}
   &SkeletalMeshLocation(
    MaskSpawnGroup=false
    Mesh Sampling Type=::NewEnumerator5
    Sampled Position Offset=&Add_Vector(
     A=<5 27 0>
     B=&MakeVector(
      X=&Add_Float(A=$Transient.XOffsetCore, B=$Transient.XOffsetRandom)
      Y=$Transient.YOffset
      Z=&FloatFromCurve(
       FloatCurve=DataInterface() {propertyValues {Curve {keys [
           {arriveTangent=-258.0267, interpMode=RCIM_Cubic, leaveTangent=-258.0267, tangentMode=RCTM_User, time=0, value=-5.5}
           <at 0.015 -9>
           <at 0.5 0>
           {arriveTangentWeight=0.5183, interpMode=RCIM_Cubic, leaveTangentWeight=0.016, tangentMode=RCTM_User, time=0.985, value=9}
           {arriveTangent=-441.0373, interpMode=RCIM_Cubic, leaveTangent=-441.0373, tangentMode=RCTM_User, time=1, value=5.5}
          ]}}}
       CurveIndex=$Transient.ZOffset
      )
     )
    )
    Skeletal Mesh=$User.Skeletal Mesh
    Socket Sampling Mode=::NewEnumerator5
   )
  ]
  particleUpdateScript [
   &ParticleState()
   &SpriteRotationRate(Rotation Rate=&FloatFromCurve(
     FloatCurve=DataInterface() {propertyValues {Curve {keys [<at 0 1>,<at 1 0.5>]}}}
     Scale Curve=&RandomRangeFloat(Minimum=-200, Maximum=200)
    ))
   &ScaleSpriteSize(Uniform Curve Sprite Scale=DataInterface() {propertyValues {Curve {keys [
        {arriveTangent=21.6257, interpMode=RCIM_Cubic, leaveTangent=21.6257, tangentMode=RCTM_User, time=0, value=0}
        {arriveTangent=-4.8392, interpMode=RCIM_Cubic, leaveTangent=-4.8392, tangentMode=RCTM_User, time=0.15, value=2}
        {arriveTangentWeight=0.1519, interpMode=RCIM_Cubic, leaveTangentWeight=0.1085, tangentMode=RCTM_User, time=0.5, value=0.75}
        {arriveTangent=1.9283, interpMode=RCIM_Cubic, leaveTangent=1.9283, tangentMode=RCTM_User, time=0.75, value=1.5}
        {arriveTangent=-6.4069, interpMode=RCIM_Cubic, leaveTangent=-6.4069, tangentMode=RCTM_User, time=1, value=0}
       ]}}})
   &CameraOffset(Camera Offset Amount=1)
   &ScaleColor(
    Scale Alpha=&FloatFromCurve(FloatCurve=DataInterface() {propertyValues {Curve {keys [
         <at 0.15 1 RCIM_Cubic>
         {arriveTangent=-0.7621, interpMode=RCIM_Cubic, leaveTangent=-0.7621, tangentMode=RCTM_User, time=0.2843, value=0.3794}
         <at 1 0.1>
        ]}}})
    ScaleA=true
    ScaleRGB=false
   )
  ]
  rendererClasses [&NiagaraSpriteRendererProperties]
  renderers [{rendererClass=&NiagaraSpriteRendererProperties}]
 }
}
summary {
 systemName=NS_MicStand_HarpTar_Idle
 emitters [
  {bEnabled=false, emitterName=DebugGlows, rendererClasses [&NiagaraSpriteRendererProperties]}
  {emitterName=Glow, rendererClasses [&NiagaraSpriteRendererProperties]}
  {bEnabled=false, emitterName=GlowPersistent, rendererClasses [&NiagaraSpriteRendererProperties]}
  {bEnabled=false, emitterName=DebugSparkle, rendererClasses [&NiagaraSpriteRendererProperties]}
  {emitterName=Sparkle, rendererClasses [&NiagaraSpriteRendererProperties]}
 ]
 userVariables [User.Skeletal Mesh {
   type=NiagaraDataInterfaceSkeletalMesh
   defaultValue {
    struct=NiagaraExt_VariableValue_DataInterface
    value {dataInterface=&self:NiagaraDataInterfaceSkeletalMesh_0}
   }
  }]
}
